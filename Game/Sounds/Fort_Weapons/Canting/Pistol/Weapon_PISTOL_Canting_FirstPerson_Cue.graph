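# Sound cue for first-person pistol canting foley, in the weapon foley sound class.
# A Random node picks between two equally weighted wave variations and feeds a Modulator that holds volume at a fixed reduced level with no variation.
# The cue uses the shared weapon reload attenuation settings.
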SoundCue /Game/Sounds/Fort_Weapons/Canting/Pistol/Weapon_PISTOL_Canting_FirstPerson_Cue
root=Modulator_0
nodes [
 {id=Modulator_0, children [Random_0], properties {volumeMax=0.7, volumeMin=0.7}}
 {id=Random_0, children [WavePlayer_1,WavePlayer_2], properties {weights [1,1]}}
 {id=WavePlayer_1, properties {soundWaveAssetPtr=&/Weapon_PISTOL_Canting_FirstPerson_01}}
 {id=WavePlayer_2, properties {soundWaveAssetPtr=&/Weapon_PISTOL_Canting_FirstPerson_02}}
]
properties {
 attenuationSettings=&Fort_Weapon_Reload_Att
 maxDistance=2250
 soundClassObject=&Fort_WeaponFoley
 volumeMultiplier=1
}
